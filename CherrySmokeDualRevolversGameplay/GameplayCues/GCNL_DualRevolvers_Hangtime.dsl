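; Looping gameplay cue for Dual Revolvers hangtime, handling its audio and VFX on the player pawn.
; On application, locally viewed players get a control bus mix or submix effect override, chosen by a hotfix flag.
; Loop start attaches 1P/3P audio, feeds remaining hangtime to Niagara, and binds an additive fire sound for local viewers.
; Removal deactivates the mix, plays an end sound, fades the loop and unbinds the fire event.
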(event EventOnLoopingStart(EitherVFX-preferthisversion)
  (params (MyTarget :type Actor%20Object%20Reference) (Parameters :type Gameplay%20Cue%20Parameters%20Structure%20%28by%20ref%29) (ParticleComponents :type Array%20of%20FXSystem%20Component%20Object%20References) (AudioComponents :type Array%20of%20Audio%20Component%20Object%20References))
  (|CallParentFunction MyTarget Parameters ParticleComponents AudioComponents)
  (bind _asfort_player_pawn (Utilities|Casting|CastToFortPlayerPawn MyTarget))
  (Variables|Default|SetAs%20Fort%20Player%20Pawn _asfort_player_pawn)
  (bind _returnvalue (Audio|SpawnSoundAttached (select (Pawn|IsLocallyViewed (Variables|Default|GetAs%20Fort%20Player%20Pawn)) (Variables|Default|GetHangtime_Loop_1P) (Variables|Default|GetHangtime_Loop_3P)) (Variables|/Script/Engine.Actor|GetRootComponent (Variables|Default|GetAs%20Fort%20Player%20Pawn))))
  (Variables|Default|SetHangTimeLoopAudio _returnvalue)
  ;[0.004565 0.2358 0.8958 0.5] Everything past this cast here assumes dual revolvers are equipped.
  ; If non-revolver equipped logic is needed, add to sequence node
  (Utilities|FlowControl|Sequence
    (:then_0
      (bind _asb_cherry_smoke_dual_revolvers_athena (Utilities|Casting|CastToB_CherrySmoke_DualRevolvers_Athena (Variables|/Script/FortniteGame.FortPawn|GetCurrentWeapon (Variables|Default|GetAs%20Fort%20Player%20Pawn))))
      (bind _output_get (Variables|Default|SetDualRevolvers _asb_cherry_smoke_dual_revolvers_athena))
      (bind _asniagara_particle_system_component (Utilities|Casting|CastToNiagaraComponent (Utilities|Array|Get(acopy) ParticleComponents))
        (:then
          (Niagara|SetNiagaraVariable(Float) _asniagara_particle_system_component "HangtimeTimeRemaining" (Variables|/CherrySmokeDualRevolversGameplay/Gameplay/Actors/B_CherrySmoke_DualRevolvers_Athena.B_CherrySmoke_DualRevolvers_Athena_C|GetHangtimeTimeRemaining _output_get)))
        (:CastFailed
          _))
      (if (Pawn|IsLocallyViewed (Variables|Default|GetAs%20Fort%20Player%20Pawn))
        (EventDispatchers|BindEventtoOnWeaponFireAudioVisuals _output_get (EventDispatchers|CreateEvent :!event Custom|WeaponFireAdditiveLayer))))))

(event Custom|WeaponFireAdditiveLayer
  (params (Weapon :type Fort%20Weapon%20Object%20Reference))
  (SoundLibrary|PlaySoundSimple (WorldSubsystems|GetSoundLibrarySubsystem (Variables|Default|GetDualRevolvers)) (Variables|Default|GetDualRevolvers) (Utilities|Struct|MakeSimpleSoundSettings (Variables|Default|GetAdditiveLayerEventName)) :!class /Script/SoundLibrary.SoundLibrarySubsystem))

(event EventOnRemoval(EitherVFX-preferthisversion)
  (params (MyTarget :type Actor%20Object%20Reference) (Parameters :type Gameplay%20Cue%20Parameters%20Structure%20%28by%20ref%29) (ParticleComponents :type Array%20of%20FXSystem%20Component%20Object%20References) (AudioComponents :type Array%20of%20Audio%20Component%20Object%20References) (BurstCameraShakeInstance :type Camera%20Shake%20Base%20Object%20Reference) (BurstDecalInstance :type Decal%20Actor%20Object%20Reference))
  (|CallParentFunction MyTarget Parameters ParticleComponents AudioComponents BurstCameraShakeInstance BurstDecalInstance)
  (Audio|Modulation|DeactivateControlBusMix MyTarget "/CherrySmokeDualRevolversGameplay/Assets/Audio/CBM_CherrySmoke_DualRevolvers.CBM_CherrySmoke_DualRevolvers")
  (bind _asfort_player_pawn (Utilities|Casting|CastToFortPlayerPawn MyTarget))
  (Audio|SpawnSoundAttached (select (Pawn|IsLocallyViewed _asfort_player_pawn) (Variables|Default|GetHangtime_End_1P) (Variables|Default|GetHangtime_End_3P)) (Variables|/Script/Engine.Actor|GetRootComponent _asfort_player_pawn))
  (Audio|Components|Audio|FadeOut (Variables|Default|GetHangTimeLoopAudio) 0.5 0.0 "Linear" :!class /Script/Engine.AudioComponent)
  (bind _dualrevolvers (Variables|Default|GetDualRevolvers)
    (:then
      (EventDispatchers|UnbindEventfromOnWeaponFireAudioVisuals _dualrevolvers (EventDispatchers|CreateEvent :!event Custom|WeaponFireAdditiveLayer)))
    (:else)))

(event EventOnApplication(EitherVFX-preferthisone)
  (params (MyTarget :type Actor%20Object%20Reference) (Parameters :type Gameplay%20Cue%20Parameters%20Structure%20%28by%20ref%29) (ParticleComponents :type Array%20of%20FXSystem%20Component%20Object%20References) (AudioComponents :type Array%20of%20Audio%20Component%20Object%20References) (BurstCameraShakeInstance :type Camera%20Shake%20Base%20Object%20Reference) (BurstDecalInstance :type Decal%20Actor%20Object%20Reference))
  (|CallParentFunction MyTarget Parameters ParticleComponents AudioComponents BurstCameraShakeInstance BurstDecalInstance)
  (bind _asfort_player_pawn (Utilities|Casting|CastToFortPlayerPawn MyTarget))
  (if (Pawn|IsLocallyViewed _asfort_player_pawn)
    (if (not (ScalableFloat|GetValueasBool (Variables|Default|GetHFDisableDualRevolverModulation)))
      (Audio|Modulation|ActivateControlBusMix _asfort_player_pawn "/CherrySmokeDualRevolversGameplay/Assets/Audio/CBM_CherrySmoke_DualRevolvers.CBM_CherrySmoke_DualRevolvers")
      (Audio|Effects|SetSubmixEffectChainOverride _asfort_player_pawn "/Game/Sounds/SubMixes/Fortnite_SFX.Fortnite_SFX" (Variables|Default|GetSubmix%20Effect%20Preset%20Chain) 1.0))))

(fn ConstructionScript
  (|CallParentFunction))
